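# Translucent UI material for Winterfest 2024 item shop tiles, building an animated glitter background with rounded-corner support.
# Shared background-parameter colors feed a radial gradient and sparkle layers.
# A snowflake sparkle texture is broken up by panning blurred noise, alongside a delayed, eased diagonal sheen sweep.
# Procedural star shapes use tiled UVs, per-cell random values and ping-pong size animation to create twinkling.
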Material /ItemShop/UI/Art/Winterfest/WF2024/Materials/M_UI_ShopTile_Glitter
graph {
 properties {
  bAutomaticallySetUsageInEditor=false
  bUsedWithStaticMesh=false
  blendMode=BLEND_Translucent
  materialDomain=MD_UI
  usedShadingModels=Default Lit
 }
 nodes [
  n1 = NamedRerouteUsage()
  n2 = Saturate(None=Multiply(A=NamedRerouteUsage(), B=ScalarParameter() {defaultValue=32, group=Sparkles, parameterName=SparklesIntensity, sortPriority=3}))
  n3 = Min(A=@n2, B=@n1)
  Add(A=Divide(A=@n1) {constB=1}, B=@n3)
  Add(A=NamedRerouteUsage(), B=Multiply(A=NamedRerouteUsage()) {constB=2})
  MaterialFunctionCall!(Input Alpha=Constant() {r=1}, bOverrideCornerRadius, Corner Radius Override) {materialFunction=/FortUILibrary/Systems/ItemShopTile/Materials/MaterialFunctions/MF_UI_ItemShop_RoundedCorners}
  n4 = Multiply(
   A=MaterialFunctionCall() {materialFunction=&MF_UI_UVScaling}.Scaled to Fit
   B=ScalarParameter() {defaultValue=16, parameterName=Tilling amount}
  ) {constB=20}
  n5 = Frac(None=@n4)
  n6 = ComponentMask(None=@n5) {r=true}
  n7 = ComponentMask(None=@n5) {g=true}
  n8 = Multiply(A=Abs(None=Subtract(A=@n6) {constB=0.5}), B=Abs(None=Subtract(A=@n7) {constB=0.5}))
  Min(
   A=OneMinus(None=SmoothStep(Value=@n8) {constMax=0.005})
   B=MaterialFunctionCall(UV=@n5, GradientStart=Constant() {r=0.3}, GradientEnd=Constant() {r=1}) {materialFunction=&MF_UI_RadialGradient}.Smooth
  )
  n9 = MaterialFunctionCall() {materialFunction=/FortUILibrary/Systems/ItemShopTile/Materials/MaterialFunctions/MF_UI_ItemShop_BackgroundParams}
  NamedRerouteDeclaration(None=@n9.Color 3) {name=Accent Color, nodeColor=<rgba 0 0 0 1>}
  NamedRerouteDeclaration(None=@n9.Color 1) {name=Primary Color, nodeColor=<rgba 0 0 0 1>}
  NamedRerouteDeclaration(None=@n9.Color 2) {name=Secondary Color, nodeColor=<rgba 0 0 0 1>}
  n10 = MaterialFunctionCall() {materialFunction=&GetUserInterfaceUV}
  NamedRerouteDeclaration(None=MaterialFunctionCall(UV=NamedRerouteDeclaration(None=@n10.Normalized UV) {name=Normalized UV, nodeColor=<rgba 0 0 0 1>}) {materialFunction=&MF_UI_UVScaling}.ScaleY) {name=Scale Y UV, nodeColor=<rgba 0 0 0 1>}
  n11 = ScalarParameter() {defaultValue=1, group=0 - Global, parameterName=Global Speed}
  NamedRerouteDeclaration(None=Multiply(A=Time(), B=@n11)) {name=Global Speed, nodeColor=<rgba 0 0 0 1>}
  n12 = NamedRerouteUsage()
  n13 = ScalarParameter() {defaultValue=0.3, group=Sparkles, parameterName=BGSparkle Opacity}
  n14 = NamedRerouteUsage()
  n15 = NamedRerouteUsage()
  n16 = Add(
   A=MaterialFunctionCall(UVs=NamedRerouteUsage(), Texture Scale=ScalarParameter() {defaultValue=1, group=01 - Gradient, parameterName=Radial Gradient Scale}) {materialFunction=&ScaleUVsByCenter}.UVs
   B=AppendVector(
    A=ScalarParameter() {group=01 - Gradient, parameterName=Radial Gradient Offset X}
    B=ScalarParameter() {group=01 - Gradient, parameterName=Radial Gradient Offset Y}
   )
  ) {constB=0.5}
  n17 = Saturate(None=MaterialFunctionCall(UV=MaterialFunctionCall(UV=@n16) {materialFunction=&MF_UI_UVScaling}.Scaled To Fill) {materialFunction=&MF_UI_RadialGradient}.Smooth)
  NamedRerouteDeclaration(None=Add(A=Multiply(A=@n12, B=@n13), B=LinearInterpolate(A=@n14, B=@n15, Alpha=@n17))) {name=Colored Gradient, nodeColor=<rgba 0 0 0 1>}
  NamedRerouteDeclaration(None=@n11) {name=Global Speed Input, nodeColor=<rgba 0 0 0 1>}
  NamedRerouteDeclaration(None=MaterialFunctionCall!(ScreenUV, TextureSize, ScalePivotPoint) {materialFunction=/FortUILibrary/MaterialFunctions/Basics/MF_UI_ScreenSpaceUVScaling}.Scaled to Fit) {name=Scale to Fill UV, nodeColor=<rgba 0 0 0 1>}
  n18 = TextureSample(UVs=MaterialFunctionCall(UVs=NamedRerouteUsage(), Texture Scale=ScalarParameter() {defaultValue=0.408, group=Sparkles, parameterName=SparkleScale, sortPriority=2}) {materialFunction=&ScaleUVsByCenter}.UVs) {texture=/Game/Catalog/BackgroundMaterialLibrary/CoreMaterials/Snowflakes_Pattern/T_Noise_Sparkle_Flake}
  n19 = TextureSample(UVs=Panner(Coordinate=NamedRerouteUsage()) {speedX=0.07, speedY=0.07}) {samplerType=SAMPLERTYPE_Grayscale, texture=&LowResBlurredNoise}
  NamedRerouteDeclaration(None=Multiply(A=NamedRerouteDeclaration(None=@n18.R) {name=StarsMask, nodeColor=<rgba 0 0 0 1>}, B=Power(Base=@n19.R))) {name=MaskedStarsMask, nodeColor=<rgba 0 0 0 1>}
  n20 = NamedRerouteUsage()
  n21 = ScalarParameter() {defaultValue=0.15, group=04 - Sheen, parameterName=Sheen Angle}
  n22 = LinearInterpolate(Alpha=MaterialFunctionCall!(
    Animation Duration=ScalarParameter() {defaultValue=2.5, group=04 - Sheen, parameterName=Sheen Duration}
    Delay=ScalarParameter() {defaultValue=3, group=04 - Sheen, parameterName=Sheen Delay}
    Reset to Default
   ) {materialFunction=/FortUILibrary/MaterialFunctions/Time/MF_UI_AnimationDelay}.Result) {constA=-1}
  n23 = Constant()
  n24 = MaterialFunctionCall(
   UVs=Add(A=Rotator(Coordinate=@n20, Time=@n21) {speed=6.2832}, B=AppendVector(A=@n22, B=@n23))
   Texture Scale=AppendVector(A=ScalarParameter() {defaultValue=0.5, group=04 - Sheen, parameterName=Sheen Scale}, B=Constant() {r=5})
  ) {materialFunction=&ScaleUVsByCenter}
  n25 = ComponentMask(None=MaterialFunctionCall!(Value=@n24.UVs, Scale=Constant() {r=1}, Height=Constant() {r=1}) {materialFunction=/FortUILibrary/MaterialFunctions/Basics/MF_UI_PingPong}.Result) {r=true}
  n26 = MaterialFunctionCall("Value (0-1)"=MaterialFunctionCall("Value (0-1)"=MaterialFunctionCall!(UVs=@n24.UVs, Alpha=@n25) {materialFunction=/FortUILibrary/MaterialFunctions/Basics/MF_UI_ClipEdgePixel}.Result) {materialFunction=&MF_UI_EaseCurves}.EaseInQuint) {materialFunction=&MF_UI_EaseCurves}
  NamedRerouteDeclaration(None=Multiply(A=Saturate(None=@n26.EaseInOutSine), B=ScalarParameter() {defaultValue=2, group=04 - Sheen, parameterName=Sheen Opacity, sliderMax=1})) {name=SheenLight, nodeColor=<rgba 0 0 0 1>}
  n27 = NamedRerouteUsage()
  n28 = AppendVector(A=ScalarParameter() {defaultValue=0.15, group=Sparkles, parameterName=SparklesSpeedSheen, sortPriority=5}, B=Constant())
  n29 = Multiply(
   A=TextureSample(UVs=Panner(Coordinate=@n27, Speed=@n28)) {samplerType=SAMPLERTYPE_Grayscale, texture=&LowResBlurredNoise}.R
   B=NamedRerouteUsage()
  )
  NamedRerouteDeclaration(None=Saturate(None=Multiply(A=@n29) {constB=2})) {name=NoiseCloudsSheenMask, nodeColor=<rgba 0 0 0 1>}
  n30 = ScalarParameter() {defaultValue=0.1, group=Sparkles, parameterName=SparklesSpeedColor, sortPriority=6}
  n31 = Constant() {r=0.03}
  n32 = TextureSample(UVs=Panner(Coordinate=NamedRerouteUsage(), Speed=AppendVector(A=@n30, B=@n31)) {speedX=0.1}) {samplerType=SAMPLERTYPE_Grayscale, texture=&LowResBlurredNoise}
  NamedRerouteDeclaration(None=Saturate(None=Multiply(A=@n32.R) {constB=3})) {name=NoiseCloudMaskSimple, nodeColor=<rgba 0 0 0 1>}
  n33 = LinearInterpolate(A=NamedRerouteUsage(), B=NamedRerouteUsage(), Alpha=@n2)
  NamedRerouteDeclaration(None=LinearInterpolate(A=LinearInterpolate(B=@n33, Alpha=@n2), B=NamedRerouteUsage(), Alpha=Multiply(A=@n3) {constB=20})) {name=SparklesSystem, nodeColor=<rgba 0 0 0 1>}
  n34 = NamedRerouteUsage()
  n35 = Constant() {r=-0.1}
  n36 = Constant() {r=0.27}
  NamedRerouteDeclaration(None=Saturate(None=MaterialFunctionCall(UVs=@n34, GradientStart=@n35, GradientEnd=@n36) {materialFunction=&MF_UI_LinearGradient}.Result)) {name=Sparks Gradient Mask, nodeColor=<rgba 0 0 0 1>}
  n37 = Floor(None=@n4)
  NamedRerouteDeclaration(None=MaterialFunctionCall!(UV=ConstantBiasScale(None=@n37), Seed=LinearInterpolate() {constA=0.5}) {materialFunction=&MF_UI_Random}.Result) {name=randomd cell vaules, nodeColor=<rgba 0.0781 0 1 1>}
  n38 = MaterialFunctionCall(UV=MaterialFunctionCall(UVs=NamedRerouteUsage(), Texture Scale=Constant() {r=1.2}) {materialFunction=&ScaleUVsByCenter}.UVs) {materialFunction=&MF_UI_UVScaling}
  n39 = NamedRerouteUsage()
  n40 = Multiply(A=Abs(None=Subtract(A=@n39) {constB=0.5})) {constB=2}
  n41 = MaterialFunctionCall(UVs=@n40) {materialFunction=&ScaleUVsByCenter}
  n42 = MaterialFunctionCall(UVs=@n40) {materialFunction=&ScaleUVsByCenter}
  n43 = OneMinus(None=LinearInterpolate(A=ComponentMask(None=@n41.UVs) {g=true}, Alpha=ComponentMask(None=@n42.UVs) {r=true}))
  n44 = Constant() {r=1}
  n45 = Constant() {r=600}
  n46 = Constant() {r=4}
  n47 = NamedRerouteUsage()
  n48 = Multiply(A=Power(Base=Multiply(A=@n43, B=@n44), Exp=LinearInterpolate(A=@n45, B=@n46, Alpha=@n47)))
  NamedRerouteDeclaration(None=Min(
    A=MaterialFunctionCall(UV=@n38.Scaled To Fill) {materialFunction=&MF_UI_RadialGradient}.Smooth
    B=Saturate(None=@n48)
   )) {name=star, nodeColor=<rgba 0 1 0.1328 1>}
  n49 = Multiply(A=Time(), B=ScalarParameter() {defaultValue=0.2, parameterName=Speed})
  n50 = NamedRerouteUsage()
  n51 = LinearInterpolate(
   A=ScalarParameter() {parameterName=MinSize}
   B=ScalarParameter() {defaultValue=0.98, parameterName=MaxSize}
   Alpha=MaterialFunctionCall("Value (0-1)"=MaterialFunctionCall(Value=@n49, Scale=@n50) {materialFunction=&MF_UI_PingPong}.Result) {materialFunction=&MF_UI_EaseCurves}.EaseInOutSine
  )
  NamedRerouteDeclaration(None=@n51) {name=Aim_Sparkle, nodeColor=<rgba 1 0 0.75 1>}
  NamedRerouteDeclaration(None=@n5) {name=Tilling UVs, nodeColor=<rgba 0 0.2734 1 1>}
  NamedRerouteDeclaration(None=@n51) {name=Aim_slow sparkle, nodeColor=<rgba 1 0 0.3281 1>}
  n52 = TextureCoordinate()
  n53 = Subtract(A=ComponentMask(None=@n52) {r=true}) {constB=0.5}
  n54 = Subtract(A=ComponentMask(None=@n52) {g=true}) {constB=0.5}
  n55 = SmoothStep(Value=Multiply(A=Abs(None=@n53), B=Abs(None=@n54))) {constMax=0.005}
  n56 = Constant() {r=0.3}
  n57 = Constant() {r=1}
  NamedRerouteDeclaration(None=Min(
    A=OneMinus(None=@n55)
    B=MaterialFunctionCall(UV=@n52, GradientStart=@n56, GradientEnd=@n57) {materialFunction=&MF_UI_RadialGradient}.Smooth
   )) {name=Stars, nodeColor=<rgba 0 1 0.625 1>}
  NamedRerouteDeclaration(None=Multiply(A=NamedRerouteUsage(), B=NamedRerouteUsage())) {name=StarShines, nodeColor=<rgba 1 0.3047 0 1>}
  NamedRerouteDeclaration(None=@n17) {name=BGRadialGradientMAsk, nodeColor=<rgba 0 0.6016 1 1>}
  NamedRerouteUsage()
  NamedRerouteUsage()
 ]
}
