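# Translucent UI material that draws animated, warm gold twinkling lights for the Winterfest 2024 Item Shop wrapper.
# Screen-space scaled UVs drive panning samples of a lights texture, broken up by subtracting a drifting blurred-noise texture.
# Time-offset sine waves along both screen axes, smoothstepped and min-combined with that mask, create moving glints.
# Nested lerps tint the result between gold, brown and cream; an Item Shop rounded-corners function is also included.
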Material /ItemShop/UI/Art/Winterfest/WF2024/Materials/M_UI_WrapperLights
graph {
 properties {
  bAutomaticallySetUsageInEditor=false
  bUsedWithStaticMesh=false
  blendMode=BLEND_Translucent
  materialDomain=MD_UI
  usedShadingModels=Default Lit
 }
 nodes [
  n1 = MaterialFunctionCall!(ScreenUV, TextureSize, ScalePivotPoint) {materialFunction=/FortUILibrary/MaterialFunctions/Basics/MF_UI_ScreenSpaceUVScaling}
  n2 = Time()
  n3 = Sine(None=Add(A=ComponentMask(None=@n1.Scaled to Fit) {g=true}, B=Multiply(A=@n2) {constB=-0.05}))
  n4 = Panner(Coordinate=MaterialFunctionCall(UVs=@n1.Scaled to Fit, Texture Scale=Constant() {r=0.25}) {materialFunction=&ScaleUVsByCenter}.UVs) {speedX=-0.01, speedY=-0.01}
  n5 = Constant() {r=0.75}
  n6 = TextureSample(UVs=Panner(Coordinate=MaterialFunctionCall(UVs=@n1.Scaled to Fit, Texture Scale=@n5) {materialFunction=&ScaleUVsByCenter}.UVs) {speedX=0.01, speedY=0.02}) {samplerType=SAMPLERTYPE_Grayscale, texture=/FortWater/Textures/LowResBlurredNoise}
  n7 = Saturate(None=Subtract(A=TextureSample(UVs=@n4) {samplerType=SAMPLERTYPE_LinearColor, texture=/ItemShop/UI/Art/Winterfest/WF2024/Textures/T_UI_Lights}.R, B=@n6.R))
  n8 = Add(A=ComponentMask(None=@n1.Scaled to Fit) {r=true}, B=Multiply(A=@n2) {constB=0.1})
  n9 = Min(A=SmoothStep(Value=Sine(None=@n8)) {constMin=-1}, B=@n7)
  n10 = Min(A=Min(A=SmoothStep(Value=@n3) {constMin=-0.5}, B=@n7), B=@n9)
  n11 = Constant3Vector() {constant=<rgba 1 0.9616 0.8247 1>}
  n12 = Add(A=Saturate(None=Multiply(A=@n10, B=@n11)), B=Multiply(A=@n9, B=Constant3Vector() {constant=<rgba 1 0.9908 0.8872 1>}))
  n13 = Constant() {r=0.25}
  n14 = TextureSample(UVs=Panner(Coordinate=MaterialFunctionCall(UVs=@n1.Scaled to Fit, Texture Scale=@n13) {materialFunction=&ScaleUVsByCenter}.UVs) {speedX=-0.02, speedY=-0.02}) {samplerType=SAMPLERTYPE_LinearColor, texture=&T_UI_Lights}
  LinearInterpolate(
   A=Constant3Vector() {constant=<rgba 1 0.7787 0.2552 1>}
   B=LinearInterpolate(A=Constant3Vector() {constant=<rgba 0.467 0.2597 0 1>}, B=Saturate(None=@n12), Alpha=@n7)
   Alpha=Saturate(None=Subtract(A=@n6.R, B=@n14.R))
  ) {constA=1}
  MaterialFunctionCall!(Input Alpha=Constant() {r=1}, bOverrideCornerRadius, Corner Radius Override) {materialFunction=/FortUILibrary/Systems/ItemShopTile/Materials/MaterialFunctions/MF_UI_ItemShop_RoundedCorners}
  TextureSample(UVs=MaterialFunctionCall() {materialFunction=&MF_UI_UVScaling}.Scaled To Fill) {texture=&T_UI_Lights}
 ]
}
